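; BP_PipeValveManager_Light is a prototype puzzle manager that lights a referenced point light once all its pipe valves are active.
; On ValveTurned, while unsolved, it runs CheckValveStates across the valve array and calls SolvedPuzzle if none is inactive.
; SolvedPuzzle flags every valve and itself as solved, then sets isLightVisible, whose RepNotify switches the light between the configured intensity and zero.
; A final validity-check loop in SolvedPuzzle has no further action.
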(event Custom|ValveTurned
  (if (not (Variables|Default|GetPuzzleSolvedAtManager))
    (bind _allactive (CallFunction|CheckValveStates))
    (if _allactive
      (CallFunction|SolvedPuzzle))))

(fn CheckValveStates
  (bind _array_element (Utilities|Array|ForEachLoopwithBreak (Variables|Default|GetAllPipeValvesLight))
    (:LoopBody
      (if (not (Variables|/Hera/Prototype/LevelDesign/Fabricio/ValveLightProto/BP_PipeValveInteractive_Light.BP_PipeValveInteractive_Light_C|GetValveActive _array_element))
        (LocalVariables|SetValveActiveLocal false)))
    (:Completed
      _))
  (return (LocalVariables|GetValveActiveLocal)))

(fn OnRep_isLightVisible
  (if (Variables|Default|GetisLightVisible)
    (Rendering|Components|Light|SetIntensity (Variables|/Script/Engine.PointLight|GetPointLightComponent (Variables|Default|Getlight)) (Variables|Default|GetLightIntensity))
    (else
      (Rendering|Components|Light|SetIntensity (Variables|/Script/Engine.PointLight|GetPointLightComponent (Variables|Default|Getlight)) 0.0))))

(fn SolvedPuzzle
  (for _array_element (Variables|Default|GetAllPipeValvesLight)
    (Variables|/Hera/Prototype/LevelDesign/Fabricio/ValveLightProto/BP_PipeValveInteractive_Light.BP_PipeValveInteractive_Light_C|SetPuzzleIsSolved true _array_element))
  (Variables|Default|SetPuzzleSolvedAtManager true)
  (Variables|Default|SetisLightVisible true)
  (for _array_element_2 (Variables|Default|GetAllPipeValvesLight)
    (Utilities|IsValid _array_element_2)))

(fn ConstructionScript)
